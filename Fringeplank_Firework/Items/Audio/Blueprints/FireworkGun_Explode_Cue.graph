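# Sound cue for the firework gun's explosion, blending close firework bursts with distant explosion layers.
# Each layer randomly picks a wave, varies pitch and volume, then applies its own attenuation and sound class before mixing.
# A quality-level switch gives the first slot eight variations per layer, while the other three share a three-variation mix.
# Concurrency overrides cap simultaneous instances, stopping the oldest.
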SoundCue /Fringeplank_Firework/Items/Audio/Blueprints/FireworkGun_Explode_Cue
root=QualityLevel_0
nodes [
 {
  id=Attenuation_0
  children [Modulator_1]
  properties {
   attenuationSettings=/Game/Sounds/Attenuation/Fort_Weapon_Rocket_Distant_Explosion_Athena_Att
   attenuationOverrides {falloffDistance=3000, attenuationShapeExtents {x=1000}}
  }
 }
 {
  id=Attenuation_3
  children [Modulator_4]
  properties {attenuationSettings=/Game/Sounds/Attenuation/Fort_Weapon_Rocket_Medium_Explosion_Athena_Att_2, attenuationOverrides {falloffDistance=1600}}
 }
 {
  id=Attenuation_6
  children [Modulator_5]
  properties {
   attenuationSettings=&Fort_Weapon_Rocket_Distant_Explosion_Athena_Att
   attenuationOverrides {falloffDistance=3000, attenuationShapeExtents {x=1000}}
  }
 }
 {
  id=Attenuation_7
  children [Modulator_0]
  properties {attenuationSettings=&Fort_Weapon_Rocket_Medium_Explosion_Athena_Att_2, attenuationOverrides {falloffDistance=1600}}
 }
 {id=Mixer_2, children [SoundClass_5,SoundClass_0], properties {inputVolume [0.8,0.4]}}
 {id=Mixer_4, children [SoundClass_3,SoundClass_2], properties {inputVolume [0.8,0.4]}}
 {id=Modulator_0, children [Random_3], properties {pitchMax=1.1, pitchMin=0.9, volumeMax=1.1, volumeMin=0.9}}
 {id=Modulator_1, children [Random_7], properties {pitchMax=1, pitchMin=0.8, volumeMax=1.1, volumeMin=0.9}}
 {id=Modulator_4, children [Random_6], properties {pitchMax=1.1, pitchMin=0.9, volumeMax=1.1, volumeMin=0.9}}
 {id=Modulator_5, children [Random_4], properties {pitchMax=1, pitchMin=0.8, volumeMax=1.1, volumeMin=0.9}}
 {id=QualityLevel_0, children [Mixer_4,Mixer_2,Mixer_2,Mixer_2]}
 {
  id=Random_3
  children [WavePlayer_3, WavePlayer_5, WavePlayer_7, WavePlayer_9, WavePlayer_10, WavePlayer_14, WavePlayer_16, WavePlayer_17]
  properties {weights [1,1,1,1,1,1,1,1]}
 }
 {
  id=Random_4
  children [WavePlayer_18, WavePlayer_19, WavePlayer_20, WavePlayer_21, WavePlayer_22, WavePlayer_23, WavePlayer_24, WavePlayer_25]
  properties {weights [1,1,1,1,1,1,1,1]}
 }
 {id=Random_6, children [WavePlayer_26, WavePlayer_27, WavePlayer_28], properties {weights [1,1,1]}}
 {id=Random_7, children [WavePlayer_29, WavePlayer_30, WavePlayer_31], properties {weights [1,1,1]}}
 {
  id=SoundClass_0
  children [Attenuation_0]
  properties {soundClassOverride=&Fort_WeaponRanged_Fire_Distant}
 }
 {
  id=SoundClass_2
  children [Attenuation_6]
  properties {soundClassOverride=&Fort_WeaponRanged_Fire_Distant}
 }
 {
  id=SoundClass_3
  children [Attenuation_7]
  properties {soundClassOverride=&Fort_WeaponRanged_Fire_3P_PSM}
 }
 {
  id=SoundClass_5
  children [Attenuation_3]
  properties {soundClassOverride=&Fort_WeaponRanged_Fire_3P_PSM}
 }
 {id=WavePlayer_10, properties {soundWaveAssetPtr=/Fringeplank_Firework/Items/Audio/Sounds/FlareGun_Fireworks_Explode_05}}
 {id=WavePlayer_14, properties {soundWaveAssetPtr=/Fringeplank_Firework/Items/Audio/Sounds/FlareGun_Fireworks_Explode_06}}
 {id=WavePlayer_16, properties {soundWaveAssetPtr=/Fringeplank_Firework/Items/Audio/Sounds/FlareGun_Fireworks_Explode_07}}
 {id=WavePlayer_17, properties {soundWaveAssetPtr=/Fringeplank_Firework/Items/Audio/Sounds/FlareGun_Fireworks_Explode_08}}
 {id=WavePlayer_18, properties {soundWaveAssetPtr=/Fringeplank_Firework/Items/Audio/Sounds/Firework_Launcher_Explosion_Distant_01}}
 {id=WavePlayer_19, properties {soundWaveAssetPtr=/Fringeplank_Firework/Items/Audio/Sounds/Firework_Launcher_Explosion_Distant_02}}
 {id=WavePlayer_20, properties {soundWaveAssetPtr=/Fringeplank_Firework/Items/Audio/Sounds/Firework_Launcher_Explosion_Distant_03}}
 {id=WavePlayer_21, properties {soundWaveAssetPtr=/Fringeplank_Firework/Items/Audio/Sounds/Firework_Launcher_Explosion_Distant_04}}
 {id=WavePlayer_22, properties {soundWaveAssetPtr=/Fringeplank_Firework/Items/Audio/Sounds/Firework_Launcher_Explosion_Distant_05}}
 {id=WavePlayer_23, properties {soundWaveAssetPtr=/Fringeplank_Firework/Items/Audio/Sounds/Firework_Launcher_Explosion_Distant_06}}
 {id=WavePlayer_24, properties {soundWaveAssetPtr=/Fringeplank_Firework/Items/Audio/Sounds/Firework_Launcher_Explosion_Distant_07}}
 {id=WavePlayer_25, properties {soundWaveAssetPtr=/Fringeplank_Firework/Items/Audio/Sounds/Firework_Launcher_Explosion_Distant_08}}
 {id=WavePlayer_26, properties {soundWaveAssetPtr=/Fringeplank_Firework/Items/Audio/Sounds/FlareGun_Fireworks_Explode_01}}
 {id=WavePlayer_27, properties {soundWaveAssetPtr=/Fringeplank_Firework/Items/Audio/Sounds/FlareGun_Fireworks_Explode_02}}
 {id=WavePlayer_28, properties {soundWaveAssetPtr=/Fringeplank_Firework/Items/Audio/Sounds/FlareGun_Fireworks_Explode_03}}
 {id=WavePlayer_29, properties {soundWaveAssetPtr=&Firework_Launcher_Explosion_Distant_01}}
 {id=WavePlayer_3, properties {soundWaveAssetPtr=&FlareGun_Fireworks_Explode_01}}
 {id=WavePlayer_30, properties {soundWaveAssetPtr=&Firework_Launcher_Explosion_Distant_02}}
 {id=WavePlayer_31, properties {soundWaveAssetPtr=&Firework_Launcher_Explosion_Distant_03}}
 {id=WavePlayer_5, properties {soundWaveAssetPtr=&FlareGun_Fireworks_Explode_02}}
 {id=WavePlayer_7, properties {soundWaveAssetPtr=&FlareGun_Fireworks_Explode_03}}
 {id=WavePlayer_9, properties {soundWaveAssetPtr=/Fringeplank_Firework/Items/Audio/Sounds/FlareGun_Fireworks_Explode_04}}
]
properties {
 maxDistance=15000
 volumeMultiplier=1
 concurrencyOverrides {maxCount=4, resolutionRule=StopOldest, volumeScale=0.5}
 concurrencySet [/Game/Sounds/Concurrency_Rules/Weapon_Explosion_Conc]
}
